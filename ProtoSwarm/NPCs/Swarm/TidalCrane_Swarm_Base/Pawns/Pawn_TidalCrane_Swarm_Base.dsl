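; Base Blueprint pawn for the TidalCrane swarm enemy, a small melee creature with randomized scale, pickaxe crit damage and death on entering deep water.
; The server picks the scale, listens for melee attack events and sets a death flag whose RepNotify triggers client effects.
; Clients drive spawn, attack and death visuals through custom primitive data, a masked death material and Niagara.
; Movement-mode audio and randomly timed voice lines start when a listener is nearby.
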(event Custom|ForceUpdateCustomDepth
  (bind _fortpawncomponent_customdepth (Variables|Default|GetFortPawnComponent_CustomDepth))
  (if (CustomDepth|IsRenderCustomDepthEnabled _fortpawncomponent_customdepth)
    (CustomDepth|Update _fortpawncomponent_customdepth 0.0 true)))

(event EventBeginPlay
  (bind _mesh (Variables|Character|GetMesh))
  ;[1 1 1] Begin Play
  (|CallParentFunction)
  (Networking|SwitchHasAuthority
    (:Authority
      (Variables|Default|SetDeathTriggered false)
      (Ability|Async|WaitGameplayEventToActor 0 (Variables|Default|GetEventMeleeAttack)
        (:then
          (CallFunction|SetupScale)
          (EventDispatchers|BindEventtoOnDied (EventDispatchers|CreateEvent :!event Custom|OnDie)))
        (:EventReceived
          (CallFunction|PawnIsAttacking))))
    (:Remote
      (CallFunction|ForceUpdateCustomDepth)
      (Utilities|Time|SetTimerbyEvent (delegate Custom|UpdateCustomDepth) (Variables|Default|GetCustom%20Depth%20Update%20Rate) true)
      (bind _asprimitive_component (Utilities|Casting|CastToPrimitiveComponent (Variables|/Script/Engine.MovementComponent|GetUpdatedComponent (Variables|Character|GetCharacterMovement))))
      (Variables|/Script/Engine.PrimitiveComponent|SetbGenerateOverlapEvents false _asprimitive_component)
      (CallFunction|PlaySoundLibraryAudio "(TagName=\"SoundLibrary.Deimos.Spawn\")")
      (CallFunction|SpawnDirtVFX)
      ; TA for Spawn
      (Rendering|Material|SetCustomPrimitiveDataFloat _mesh 0 (Utilities|Time|GetGameTimeinSeconds))
      (Rendering|Material|SetCustomPrimitiveDataFloat _mesh 2 0.0)
      (bind _returnvalue (Rendering|Material|CreateDynamicMaterialInstance (Variables|Default|GetMaskedDeathMaterial) "None" "None" :!class /Script/Engine.KismetMaterialLibrary))
      (Variables|Default|SetMID_MaskedDeathMaterial _returnvalue)
      ; Start showing AV
      (SoundIndicator|StartSoundIndicatorLooping (Variables|SoundIndicator|GetSoundIndicatorComponent) (Transformation|GetActorLocation) "/ProtoSwarm/Assets/Audio/TidalCrane_Swarmer/MSS_TidalCrane_Movement_Swarmer_Single.MSS_TidalCrane_Movement_Swarmer_Single" 1.0 0))))

(event Custom|UpdateCustomDepth
  (bind _fortpawncomponent_customdepth (Variables|Default|GetFortPawnComponent_CustomDepth))
  ;[1 1 1] Begin Play
  (if (CustomDepth|IsRenderCustomDepthEnabled _fortpawncomponent_customdepth)
    (CustomDepth|Update _fortpawncomponent_customdepth (Variables|Default|GetCustom%20Depth%20Update%20Rate) false)))

(event GameplayCue|GameplayCue.Impact.Physical.Creature
  (params (EventType :type EGameplayCueEvent%20Enum :default "OnActive") (Parameters :type Gameplay%20Cue%20Parameters%20Structure%20%28by%20ref%29))
  (bind (_bblockinghit _binitialoverlap _time _distance _location) (Collision|BreakHitResult (Ability|GameplayCue|GetHitResult Parameters)))
  ;[1 1 1] ImpactATarget
  (Audio|PlaySoundatLocation (Sound|GetImpactPhysicalSurfaceSound (Ability|GameplayCue|GetHitResult Parameters)) _location)
  (Default|PlayDamageFXSurfaceEnemies Parameters (Variables|Effects|GetImpactParticles_Default) (Variables|Effects|GetImpactParticles_Wood) (Variables|Effects|GetImpactParticles_Stone) (Variables|Effects|GetImpactParticles_Metal) (Variables|Effects|GetImpactParticles_HumanFlesh) (.location (Transformation|GetSocketTransform (Variables|Character|GetMesh) (Variables|Default|GetImpact_MeleeSocket) "RTS_World")) (.rotation (Transformation|GetSocketTransform (Variables|Character|GetMesh) (Variables|Default|GetImpact_MeleeSocket) "RTS_World")) (Variables|Default|GetMeleeImpact_UseSocketTransforms)))

(event Custom|OnDeathTriggered
  (bind _mesh (Variables|Character|GetMesh))
  ; On Death Effects
  (Networking|SwitchHasAuthority
    (:Remote
      (Lighting|SetCastCapsuleIndirectShadow (Variables|Character|GetMesh) false)
      (if (ScalableFloat|GetValueasBool (Variables|Default|GetPlayBurstVFXOnElim))
        (bind _returnvalue (Niagara|SpawnSystematLocation (Variables|Default|GetFX_BugDeath) (Transformation|GetActorLocation) (Transformation|GetActorRotation)))
        (Niagara|SetNiagaraVariable(Float) _returnvalue "FXScale" (Variables|Default|GetFX_BugDeath_Scale))
        (Niagara|SetNiagaraVariable(Bool) _returnvalue "FXDecal" (Variables|Default|GetFX_BugDeath_Decal))
        (else
          _))
      (Utilities|FlowControl|Sequence
        (:then_0
          (CallFunction|PlaySoundLibraryAudio "(TagName=\"SoundLibrary.Deimos.Death\")")
          (CallFunction|FadeOutAudio))
        (:then_1
          (Audio|PlaySoundatLocation (Variables|Sound|GetDeath_Dematerialize_Sound) (Transformation|GetActorLocation self)) ; Dematerialize Sound
          (if (ScalableFloat|GetValueasBool (Variables|Default|GetPlayFadeVFXOnElim))
            (Rendering|Material|SetMaterial _mesh 0 (Variables|Default|GetMID_MaskedDeathMaterial)) ; Swap to masked material on death
            (Rendering|Material|SetCustomPrimitiveDataFloat _mesh 2 1.0)
            (Rendering|Material|SetCustomPrimitiveDataFloat _mesh 3 (Utilities|Time|GetGameTimeinSeconds))
            (Utilities|FlowControl|Delay 1.0)
            (else
              _))
          (Development|SetHiddeninGame (Variables|Character|GetMesh) true false)
          (SoundIndicator|StopSoundIndicatorLooping (Variables|SoundIndicator|GetSoundIndicatorComponent)))))))

(event EventOnDeathPlayEffects
  (params (Damage :type Float%20%28single-precision%29) (DamageTags :type Gameplay%20Tag%20Container%20Structure%20%28by%20ref%29) (Momentum :type Vector :default "0, 0, 0") (HitInfo :type Hit%20Result%20Structure%20%28by%20ref%29) (InstigatedBy :type Fort%20Pawn%20Object%20Reference) (DamageCauser :type Actor%20Object%20Reference) (EffectContext :type Gameplay%20Effect%20Context%20Handle%20Structure))
  ; On Death Effects
  (|CallParentFunction Damage DamageTags Momentum HitInfo InstigatedBy DamageCauser EffectContext))

(event EventOnDamagePlayEffects
  (params (Damage :type Float%20%28single-precision%29) (DamageTags :type Gameplay%20Tag%20Container%20Structure%20%28by%20ref%29) (Momentum :type Vector :default "0, 0, 0") (HitInfo :type Hit%20Result%20Structure%20%28by%20ref%29) (InstigatedBy :type Fort%20Pawn%20Object%20Reference) (DamageCauser :type Actor%20Object%20Reference) (EffectContext :type Gameplay%20Effect%20Context%20Handle%20Structure))
  ; Hit Effects
  (CallFunction|PlaySoundLibraryAudio "(TagName=\"SoundLibrary.Deimos.Pain\")"))

(event Custom|PawnIsAttacking
  ; Fired when the pawn is playing the Attack Animation
  (Networking|CanExecuteCosmeticEvents
    (:True
      ; TA for Attack
      (Rendering|Material|SetCustomPrimitiveDataFloat (Variables|Character|GetMesh) 5 (Utilities|Time|GetGameTimeinSeconds))
      (Default|CallOnSwarmerAttack)
      (CallFunction|RestartVOTimer)
      (bind _returnvalue (Audio|SpawnSoundAttached (Variables|Default|GetAttack%20Sound) (Variables|Transformation|GetRootComponent)))
      (Variables|Default|SetAttack%20Audio _returnvalue))))

(event WaterBody|EventOnWaterDeepNavMeshEnter
  ; Water
  (CallFunction|OnDeepWaterEnter))

(event EventOnMovementModeChanged
  (params (PrevMovementMode :type EMovementMode%20Enum :default "MOVE_None") (NewMovementMode :type EMovementMode%20Enum :default "MOVE_None") (PrevCustomMode :type Byte) (NewCustomMode :type Byte))
  (bind _movement_mode (Variables|Default|GetMovement%20Mode))
  (bind _flying_sound (Variables|Sound|GetFlying%20Sound))
  (bind _walking_sound (Variables|Sound|GetWalking%20Sound))
  (bind _returnvalue (Utilities|Select :Index _movement_mode :MOVE_Walking _walking_sound :MOVE_NavWalking _walking_sound :MOVE_Falling _flying_sound :MOVE_Flying _flying_sound))
  (|CallParentFunction PrevMovementMode NewMovementMode PrevCustomMode NewCustomMode)
  (Variables|Default|SetMovement%20Mode NewMovementMode)
  (bind (_listenerposition _returnvalue_2) (Audio|GetClosestListenerLocation (Transformation|GetActorLocation) 4.8e+03 true))
  (if _returnvalue_2
    (Utilities|FlowControl|Sequence
      (:then_0
        ; Movement audio - walking and flying
        (if (Audio|Components|Audio|IsPlaying (Variables|Default|GetMovementAudio))
          (Audio|Components|Audio|FadeOut (Variables|Default|GetMovementAudio) 0.1 0.0 "Linear" :!class /Script/Engine.AudioComponent)
          (EventDispatchers|BindEventtoOnAudioFinished (Variables|Default|GetMovementAudio) (delegate Custom|Change%20Movement%20Sound))
          (else
            (Utilities|IsValid _returnvalue
              (:Is%20Valid
                (Audio|Components|Audio|SetSound (Variables|Default|GetMovementAudio) _returnvalue)
                (Audio|Components|Audio|FadeIn (Variables|Default|GetMovementAudio) 0.1 1.0 0.0 "Linear" :!class /Script/Engine.AudioComponent))))))
      (:then_1
        (CallFunction|StartRandomVoiceTimer)))
    (else
      (Utilities|Time|ClearandInvalidateTimerbyHandle (Variables|Sound|VO|GetPlay%20VO%20Timer)))))

(event Custom|Change%20Movement%20Sound
  (bind _movement_mode (Variables|Default|GetMovement%20Mode))
  (bind _flying_sound (Variables|Sound|GetFlying%20Sound))
  (bind _walking_sound (Variables|Sound|GetWalking%20Sound))
  (bind _returnvalue (Utilities|Select :Index _movement_mode :MOVE_Walking _walking_sound :MOVE_NavWalking _walking_sound :MOVE_Falling _flying_sound :MOVE_Flying _flying_sound))
  ; Movement audio - walking and flying
  (Utilities|IsValid _returnvalue
    (:Is%20Valid
      (Audio|Components|Audio|SetSound (Variables|Default|GetMovementAudio) _returnvalue)
      (Audio|Components|Audio|FadeIn (Variables|Default|GetMovementAudio) 0.1 1.0 0.0 "Linear" :!class /Script/Engine.AudioComponent))))

(event Custom|Fade%20Out%20Audio
  (Utilities|FlowControl|Sequence
    (:then_0
      (bind _vo_audio_component (Variables|Default|GetVO%20Audio%20Component)
        (:then
          (Audio|Components|Audio|FadeOut _vo_audio_component 0.1 0.0 "Linear" :!class /Script/Engine.AudioComponent))
        (:else)))
    (:then_1
      (bind _attack_audio (Variables|Default|GetAttack%20Audio)
        (:then
          (Audio|Components|Audio|FadeOut _attack_audio 0.1 0.0 "Linear" :!class /Script/Engine.AudioComponent))
        (:else)))
    (:then_2
      (bind _movementaudio (Variables|Default|GetMovementAudio)
        (:then
          (Audio|Components|Audio|FadeOut _movementaudio 0.1 0.0 "Linear" :!class /Script/Engine.AudioComponent))
        (:else)))))

(event EventEndPlay
  (params (EndPlayReason :type EEndPlayReason%20Enum :default "Destroyed"))
  (|CallParentFunction EndPlayReason)
  (Utilities|Time|ClearandInvalidateTimerbyHandle (Variables|Sound|VO|GetPlay%20VO%20Timer))
  (EventDispatchers|UnbindEventfromOnDied (delegate Custom|OnDie)))

(event Custom|Start%20Random%20Voice%20Timer
  (bind (_returnvalue_min _returnvalue_max) (Utilities|Struct|BreakFloatInterval (select (Variables|Sound|VO|GetIs%20Initial%20Delay) (Variables|Sound|VO|GetInitial%20Delay%20Range) (Variables|Sound|VO|GetRandom%20Delay%20Range))))
  ; Voice - plays at random intervals depending on movement speed
  (bind _returnvalue (Utilities|Time|SetTimerbyEvent (EventDispatchers|CreateEvent :!event Custom|Play%20VO) (Math|Random|RandomFloatinRange _returnvalue_min _returnvalue_max)))
  (Variables|Sound|VO|SetPlay%20VO%20Timer _returnvalue)
  (Variables|Sound|VO|SetIs%20Initial%20Delay false))

(event Custom|OnDie
  (params (DamagedActor :type Actor%20Object%20Reference) (Damage :type Float%20%28single-precision%29) (InstigatedBy :type Controller%20Object%20Reference) (DamageCauser :type Actor%20Object%20Reference) (HitLocation :type Vector) (FHitComponent :type Primitive%20Component%20Object%20Reference) (BoneName :type Name) (Momentum :type Vector))
  (Variables|Default|SetDeathTriggered true))

(event Custom|Play%20VO
  ; Voice - plays at random intervals depending on movement speed
  (bind _returnvalue (Audio|SpawnSoundAttached :Sound (Variables|Sound|GetVoice%20Sound) :AttachToComponent (Variables|Transformation|GetRootComponent) :AttachPointName "head" :LocationType "KeepRelativeOffset" :bStopWhenAttachedToDestroyed true :VolumeMultiplier 1.0 :PitchMultiplier 1.0 :StartTime 0.0 :bAutoDestroy true))
  (Variables|Default|SetVO%20Audio%20Component _returnvalue)
  (CallFunction|StartRandomVoiceTimer))

(event Custom|Restart%20VO%20Timer
  ; Voice - plays at random intervals depending on movement speed
  (Utilities|Time|ClearandInvalidateTimerbyHandle (Variables|Sound|VO|GetPlay%20VO%20Timer))
  (Variables|Sound|VO|SetIs%20Initial%20Delay true)
  (CallFunction|StartRandomVoiceTimer))

(fn ApplyAwakenEffectToSkeletalMesh
  (params (Mesh :type Skeletal%20Mesh%20Component%20Object%20Reference)))

(fn ApplyAwakenEffectToStaticMesh
  (params (Mesh :type Static%20Mesh%20Component%20Object%20Reference))
  (bind _awakenfxstaticmeshes (Variables|Effects|GetAwakenFXStaticMeshes))
  (bind (_value _returnvalue) (Utilities|Map|Find _awakenfxstaticmeshes Mesh))
  (bind _fxstaticmesh (LocalVariables|GetFXStaticMesh))
  (Utilities|IsValid Mesh
    (:Is%20Valid
      (if _returnvalue
        (bind _output_get (LocalVariables|SetFXStaticMesh _value))
        (Utilities|IsValid _output_get
          (:Is%20Valid
            _)
          (:Is%20Not%20Valid
            (bind _returnvalue_2 (AddComponent|NODE_AddFXStaticMeshComponent-2))
            (Utilities|Map|Add (Variables|Effects|GetAwakenFXStaticMeshes) Mesh _returnvalue_2)
            (bind _output_get_2 (LocalVariables|SetFXStaticMesh _returnvalue_2))
            (Class|FortFXStaticMeshComponent|SetSource _output_get_2 Mesh)
            (Variables|/Script/FortniteGame.FortFXStaticMeshComponent|SetAwakenDuration (Variables|Effects|GetAwakenDuration) _fxstaticmesh)))
        (Class|FXStaticMeshComponent|ApplyAwakenEffect _fxstaticmesh)
        (else
          _)))))

(fn ApplyBuildingHitEffectToSkeletalMesh
  (params (Mesh :type Skeletal%20Mesh%20Component%20Object%20Reference)))

(fn ApplyBuildingHitEffectToStaticMesh
  (params (Mesh :type Static%20Mesh%20Component%20Object%20Reference))
  (bind _buildinghitfxstaticmeshes (Variables|Effects|GetBuildingHitFXStaticMeshes))
  (bind (_value _returnvalue) (Utilities|Map|Find _buildinghitfxstaticmeshes Mesh))
  (bind _fxstaticmesh (LocalVariables|GetFXStaticMesh))
  (Utilities|IsValid Mesh
    (:Is%20Valid
      (if _returnvalue
        (bind _output_get (LocalVariables|SetFXStaticMesh _value))
        (Utilities|IsValid _output_get
          (:Is%20Valid
            _)
          (:Is%20Not%20Valid
            (bind _returnvalue_2 (AddComponent|NODE_AddFXStaticMeshComponent-0))
            (Utilities|Map|Add (Variables|Effects|GetBuildingHitFXStaticMeshes) Mesh _returnvalue_2)
            (bind _output_get_2 (LocalVariables|SetFXStaticMesh _returnvalue_2))
            (Class|FortFXStaticMeshComponent|SetSource _output_get_2 Mesh)
            (Variables|/Script/FortniteGame.FortFXStaticMeshComponent|SetBuildingHitDuration (Variables|Effects|GetBuildingHitDuration) _fxstaticmesh)))
        (Class|FXStaticMeshComponent|ApplyBuildingHitEffect _fxstaticmesh)
        (else
          _)))))

(fn BlueprintModifyIncomingDamage
  (params (Damage :type Float%20%28single-precision%29) (DamageTags :type Gameplay%20Tag%20Container%20Structure%20%28by%20ref%29) (EffectContext :type Gameplay%20Effect%20Context%20Handle%20Structure%20%28by%20ref%29) (EventInstigator :type Controller%20Object%20Reference) (DamageCauser :type Actor%20Object%20Reference))
  (bind _returnvalue (GameplayTags|HasTag DamageTags "(TagName=\"Weapon.Melee.Impact.Pickaxe\")"))
  (bind _returnvalue_2 (|CallParentFunction Damage DamageTags EffectContext EventInstigator DamageCauser))
  (if _returnvalue
    (Ability|EffectContext|SetCritical EffectContext _returnvalue)
    (return (* (ScalableFloat|GetValueatLevel (Variables|Default|GetPickAxeCritMultiplier)) _returnvalue_2))
    (else
      (return _returnvalue_2))))

(fn FindCloseGroundPointOnNavMesh
  (bind _self self)
  (bind _returnvalue (Transformation|GetActorLocation _self))
  (bind (_outhit _returnvalue_2) (Collision|LineTraceByChannel _returnvalue "0.000000,0.000000,-10000.000000" "TraceTypeQuery10" true))
  (if _returnvalue_2
    (LocalVariables|SetIsWorldTraceHit true)
    (LocalVariables|SetWorldTraceLocation (.Location (Collision|BreakHitResult _outhit)))
    (else
      (LocalVariables|SetIsWorldTraceHit false)
      (LocalVariables|SetWorldTraceLocation _returnvalue)))
  (return (LocalVariables|GetWorldTraceLocation) (LocalVariables|GetIsWorldTraceHit)))

(fn OnDeepWaterEnter
  (Networking|SwitchHasAuthority
    (:Authority
      (FortPawn|ForceKill (Variables|Default|GetEnter%20Deep%20Water%20Death%20Reason) (Pawn|GetController) self))))

(fn OnRep_DeathTriggered
  (if (Variables|Default|GetDeathTriggered)
    (CallFunction|OnDeathTriggered)))

(fn OnRep_SelectedPawnScaleSizeIndex
  (bind _pawnscalesizes (Variables|Default|GetPawnScaleSizes))
  (bind _selectedpawnscalesizeindex (Variables|InternalData|GetSelectedPawnScaleSizeIndex))
  (Networking|SwitchHasAuthority
    (:Remote
      (if (Utilities|Array|IsValidIndex _pawnscalesizes _selectedpawnscalesizeindex)
        (Transformation|SetWorldScale3D (Variables|Character|GetCapsuleComponent) (Math|Vector|MakeVector (Utilities|Array|Get(aref) _pawnscalesizes _selectedpawnscalesizeindex) (Utilities|Array|Get(aref) _pawnscalesizes _selectedpawnscalesizeindex) (Utilities|Array|Get(aref) _pawnscalesizes _selectedpawnscalesizeindex)))))))

(fn Orphaned
  (return false 0))

(fn PlayAdditiveHitReacts
  (params (Hit%20Direction :type Vector) (Anim%20Montage :type Anim%20Montage%20Object%20Reference))
  (HitReacts|PlayAdditiveHitReactsNative Hit%20Direction Anim%20Montage) ; This used to be done directly in the Event graph, which meant any Child of this Pawn couldn't override the Function behavior.\n\nWith this, we should be able to override PlayAdditiveHitReacts and call the Native function here with different logic in front of it, if we need to.
)

(fn PlaySoundLibraryAudio
  (params (Event%20Name :type Gameplay%20Tag%20Structure))
  (Networking|CanExecuteCosmeticEvents
    (:True
      (if (Utilities|IsValid (Variables|Sound|GetSoundLibraryContext) :!class /Script/Engine.KismetSystemLibrary)
        (Utilities|Struct|SetmembersinSimpleSoundSettings (Variables|/Script/SoundLibrary.SoundLibrarySimpleContext|GetSettings (Variables|Sound|GetSoundLibraryContext)) Event%20Name)
        (else
          (bind _returnvalue (SoundLibrary|CreateSoundLibraryContext :Class "/Script/SoundLibrary.SoundLibrarySimpleContext" :Settings (Utilities|Struct|MakeSimpleSoundSettings Event%20Name) :Actor self :bForwardToActorOwner true :bForwardToAttachedActors true :bAutoResetContext true))
          (bind _output_get (Variables|Sound|SetSoundLibraryContext _returnvalue))))
      (SoundLibrary|PlaySoundwithContext (WorldSubsystems|GetSoundLibrarySubsystem) _output_get))))

(fn SetScalarParameterOnAllCharacterMIDs
  (params (Parameter%20Name :type Name) (Scalar%20Value :type Float%20%28double-precision%29) (AlsoModifyOriginalMIDs :type Boolean))
  (bind _auxillary_objects (Variables|Effects|GetAuxillary%20Objects))
  (bind _duplicate_auxillary_objects (Variables|Effects|GetDuplicate%20Auxillary%20Objects))
  (Utilities|FlowControl|Sequence
    (:then_0
      (bind _mesh (Variables|Character|GetMesh)
        (:then
          (Rendering|Material|SetScalarParameterValueonMaterials _mesh (Variables|GetParameter%20Name) (Variables|GetScalar%20Value)))
        (:else
          _))
      (bind _duplicatecharactermesh (Variables|Effects|GetDuplicateCharacterMesh)
        (:then
          (Rendering|Material|SetScalarParameterValueonMaterials _duplicatecharactermesh (Variables|GetParameter%20Name) (Variables|GetScalar%20Value)))
        (:else)))
    (:then_1
      (if (Utilities|IsArrayNonEmpty _auxillary_objects)
        (for _array_element _auxillary_objects
          (Utilities|IsValid _array_element
            (:Is%20Valid
              (bind _asmesh_component (Utilities|Casting|CastToMeshComponent _array_element))
              (Rendering|Material|SetScalarParameterValueonMaterials _asmesh_component (Variables|GetParameter%20Name) (Variables|GetScalar%20Value)))))))
    (:then_2
      (if (Utilities|IsArrayNonEmpty _duplicate_auxillary_objects)
        (for _array_element_2 _duplicate_auxillary_objects
          (Utilities|IsValid _array_element_2
            (:Is%20Valid
              (bind _asmesh_component_2 (Utilities|Casting|CastToMeshComponent _array_element_2))
              (Rendering|Material|SetScalarParameterValueonMaterials _asmesh_component_2 (Variables|GetParameter%20Name) (Variables|GetScalar%20Value)))))))
    (:then_3
      ;[1 1 1] Optionally keep the original MIDs in sync with the override MIDs
      (if AlsoModifyOriginalMIDs
        (Utilities|FlowControl|Sequence
          (:then_0
            (for _array_element_3 (Variables|Materials|GetPrevious%20MID)
              (Rendering|Material|SetScalarParameterValue _array_element_3 (Variables|GetParameter%20Name) (Variables|GetScalar%20Value) :!class /Script/Engine.MaterialInstanceDynamic)))
          (:then_1
            (for _array_element_4 (Variables|Materials|GetPrevious%20MID_AuxObjs)
              (Rendering|Material|SetScalarParameterValue _array_element_4 (Variables|GetParameter%20Name) (Variables|GetScalar%20Value) :!class /Script/Engine.MaterialInstanceDynamic))))))))

(fn SetupScale
  (bind _output_get (Variables|InternalData|SetSelectedPawnScaleSizeIndex (- (Math|Random|RandomIntegerinRange 1 (Utilities|Array|Length (Variables|Default|GetPawnScaleSizes))) 1)))
  (Transformation|SetWorldScale3D (Variables|Character|GetCapsuleComponent) (Math|Vector|MakeVector (Utilities|Array|Get(aref) (Variables|Default|GetPawnScaleSizes) _output_get) (Utilities|Array|Get(aref) (Variables|Default|GetPawnScaleSizes) _output_get) (Utilities|Array|Get(aref) (Variables|Default|GetPawnScaleSizes) _output_get))))

(fn SetVectorParameterOnAllCharacterMIDs
  (params (Parameter%20Name :type Name) (Vector%20Value :type Vector) (Propagate%20to%20Auxiliary%20Meshes :type Boolean :default true))
  (bind _auxillary_objects (Variables|Effects|GetAuxillary%20Objects))
  (bind _duplicate_auxillary_objects (Variables|Effects|GetDuplicate%20Auxillary%20Objects))
  (bind _mesh (Variables|Character|GetMesh)
    (:then
      (Rendering|Material|SetVectorParameterValueonMaterials _mesh Parameter%20Name Vector%20Value))
    (:else
      _))
  (bind _duplicatecharactermesh (Variables|Effects|GetDuplicateCharacterMesh)
    (:then
      (Rendering|Material|SetVectorParameterValueonMaterials _duplicatecharactermesh Parameter%20Name Vector%20Value))
    (:else
      _))
  (if (and (> (Utilities|Array|Length _auxillary_objects) 0) (Variables|GetPropagate%20to%20Auxiliary%20Meshes))
    (for _array_element _auxillary_objects
      (Utilities|IsValid _array_element
        (:Is%20Valid
          (bind _asmesh_component (Utilities|Casting|CastToMeshComponent _array_element))
          (Rendering|Material|SetVectorParameterValueonMaterials _asmesh_component (Variables|GetParameter%20Name) (Variables|GetVector%20Value)))))
    (if (and (> (Utilities|Array|Length _duplicate_auxillary_objects) 0) (Variables|GetPropagate%20to%20Auxiliary%20Meshes))
      (for _array_element_2 _duplicate_auxillary_objects
        (Utilities|IsValid _array_element_2
          (:Is%20Valid
            (bind _asmesh_component_2 (Utilities|Casting|CastToMeshComponent _array_element_2))
            (Rendering|Material|SetVectorParameterValueonMaterials _asmesh_component_2 (Variables|GetParameter%20Name) (Variables|GetVector%20Value)))))))
  (return))

(fn SpawnActorOnDeath)

(fn SpawnDirtVFX
  (bind _returnvalue (Niagara|SpawnSystematLocation (Variables|Default|GetSpawnInVFX) (Transformation|GetActorLocation) (Transformation|GetActorRotation) "1.000000,1.000000,1.000000" true true "None" false)) ; Play Dirt VFX
  (Utilities|IsValid _returnvalue
    (:Is%20Valid
      (Niagara|SetNiagaraVariable(Float) _returnvalue "BurrowTime" 1.0)
      (Niagara|SetNiagaraVariable(Float) _returnvalue "Scale" 1.0))))

(fn SpawnParticleSystemOnCharacterMesh
  (params (ParticleSystemTemplate :type Particle%20System%20Object%20Reference) (ParticleSystemComponentReferenceVar :type Cascade%20Particle%20System%20Component%20Object%20Reference) (AttachPointName :type Name) (Location :type Vector) (Rotation :type Rotator) (InstanceParameters :type Array%20of%20Particle%20Sys%20Param%20Structures) (AutoActivate :type Boolean :default true) (AutoDestroy :type Boolean) (AbsoluteLocation :type Boolean) (AbsoluteRotation :type Boolean) (AbsoluteScale :type Boolean))
  (Networking|CanExecuteCosmeticEvents
    (:True
      (Utilities|IsValid ParticleSystemTemplate
        (:Is%20Valid
          (Utilities|IsValid ParticleSystemComponentReferenceVar
            (:Is%20Valid
              (Components|Activation|Deactivate ParticleSystemComponentReferenceVar))
            (:Is%20Not%20Valid
              _))
          (bind _returnvalue (Effects|SpawnEmitterAttached ParticleSystemTemplate (Variables|Character|GetMesh) AttachPointName Location Rotation "1.000000,1.000000,1.000000" "KeepRelativeOffset" AutoDestroy))
          (Variables|/Script/CascadeRuntime.ParticleSystemComponent|SetInstanceParameters InstanceParameters _returnvalue)
          (Components|Activation|SetActive _returnvalue AutoActivate false)
          (Transformation|SetAbsolute _returnvalue AbsoluteLocation AbsoluteRotation AbsoluteScale)
          (return _returnvalue))))))

(fn ConstructionScript
  (Networking|CanExecuteCosmeticEvents
    (:True
      (AI|SetWeaponCollisionComponent (Variables|Default|GetWeaponCapsuleCollision))
      (bind _returnvalue (Rendering|Material|GetScalarParameterValue "/Game/Packages/Fortress_SharedMaterials/GlobalMaterialParameters/FortniteMaterialParameters.FortniteMaterialParameters" "Enemy Indirect Shadow Min Occlusion" :!class /Script/Engine.KismetMaterialLibrary))
      (Lighting|SetCapsuleIndirectShadowMinVisibility (Variables|Character|GetMesh) _returnvalue))))
